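; StateTree condition for the Tidal Crane Toy Fox boss that passes while a giant is present to fight and the match is still running.
; A giant counts as alive if any FortGiantPlayerPawn exists in the world, found by an all-actors search on each test, with no health check.
; The match counts as ended when the Athena game state is valid and has a winning team above zero.
; bInverse negates the combined result.
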(fn HasMatchEnded
  (Utilities|FlowControl|Sequence
    (:then_0
      (bind _returnvalue (Fortnite|Athena|GameState|GetGameStateAthena))
      (Utilities|IsValid _returnvalue
        (:Is%20Valid
          (return (> (Variables|/Script/FortniteGame.FortGameStateAthena|GetWinningTeam _returnvalue) 0)))))
    (:then_1
      (return false))))

(fn IsGiantAlive
  (bind _outactors (Actor|GetAllActorsOfClass "/Script/GiantPawnCoreRuntime.FortGiantPlayerPawn"))
  (return (> (Utilities|Array|Length _outactors) 0)))

(fn ReceiveTestCondition
  (bind _bended (CallFunction|HasMatchEnded))
  (bind _breturn (CallFunction|IsGiantAlive))
  (if (Variables|Default|GetbInverse)
    (return (not (and _breturn (not _bended))))
    (else
      (return (and _breturn (not _bended))))))
