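; StateTree task for the TidalCrane ToyFox boss, named for toggling giant perception, though no perception logic is present in the graph.
; On entering the state, a sequence first runs a validity check on the boss variable, with both branches left empty.
; It then immediately finishes the task as failed, so it acts as an unimplemented stub.
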(event EventEnterState
  (params (Transition :type State%20Tree%20Transition%20Result%20Structure%20%28by%20ref%29))
  (Utilities|FlowControl|Sequence
    (:then_0
      (Variables|Default|GetTidalCraneToyFoxBoss
        (:then)
        (:else)))
    (:then_1
      (StateTree|FinishTask false))))
